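# Niagara system for the knockback weapon's PlayerDampen effect, covering a skeletal mesh in glowing sprites and electric arcs.
# Decals bursts wave-tinted sprites, while Lightning_Big and Lightning_small spawn short-lived, randomly oriented meshes with particle lights.
# These emitters resample User.SKM_02 during particle update and draw mainly on User color parameters.
# Several other visible emitters are disabled, including Decals001, Mesh_Arcs, Ring_Burst001 and a component-renderer PostProcess.
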NiagaraSystem /Skyfire/Gameplay/KnockbackWeapon/FX/PlayerDampen/NS_Knockback_PlayerDampen
dependencies {
 usedDataInterfaces [&NiagaraDataInterfaceCurve]
 usedDynamicInputs []
 usedModules [
  &self:SystemSpawnScript.SystemScriptSource.SystemScriptGraph.NiagaraNodeAssignment_0.SetVariables_#1
  &SystemState
 ]
}
emitters {
 Decals {
  emitterSpawnScript [/SetVariables(Emitter.UseParticleColorInput=1) {isSetParameters=true, module=SetVariables_#2}]
  emitterUpdateScript [
   &EmitterState()
   &SpawnBurst_Instantaneous(Spawn Count=20, Use Spawn Probability=false)
   &SpawnBurst_Instantaneous(Spawn Count=6, Spawn Time=0.5, Use Spawn Probability=false) {enabled=false, module=SpawnBurst_Instantaneous001}
  ]
  particleSpawnScript [
   &InitializeParticle(
    Color=$User.Color_01
    Lifetime=1.4
    Mesh Scale Mode=::NewEnumerator3
    Mesh Uniform Scale=0.04
    Sprite Size Mode=::NewEnumerator1
    Uniform Sprite Size Max=80
    Uniform Sprite Size Min=70
    UsePositionOffset=false
    Write Material Random=true
   )
   &CameraOffset(Camera Offset Amount=12)
   &FN_SetSpawnGroup(Number of Spawn Groups=6)
   &InitialMeshOrientation(
    Orientation Axis=<0 0 1>
    Orientation Coordinate Space=::World
    Orientation Vector=&Lerp_Vector(
     A=<-1 0 0>
     B=<1 0 0>
     Alpha=&ModuloFloat(A=3, B=&ModuloFloat(A=3, B=&MakeFloatFromInt(Int=$Particles.SpawnGroup)))
    )
    Use Orientation Vector=true
    Use Rotation Vector=false
   )
  ]
  particleUpdateScript [
   &ParticleState()
   &SkeletalMeshLocation(Bone Sampling Mode=::NewEnumerator0, MaskSpawnGroup=false, OverrideRandomSeed=false, Skeletal Mesh=$User.SKM_02)
   /SetVariables(Particles.SizeWave:NiagaraFloat=&Waveform(
     Global Input Position=&Multiply_Float(A=$Engine.Time, B=3)
     Global Phase=$Particles.NormalizedAge
     "Waveform [1]"=::NewEnumerator0
     "[1] Phase Scale"=0
     "[1] Phase"=$Particles.NormalizedAge
     "[1] Frequency"=2
     "[1] Amplitude Min/Max"=<0.2 0>
     "[1] Bias"=1
     "[2] Phase"=$Emitter.Age
     "[3] Phase"=$Emitter.Age
    )) {isSetParameters=true, module=SetVariables_#3}
   /SetVariables() {isSetParameters=true, module=SetVariables_#4}
   /SetVariables(Particles.WaveColor:LinearColor=&MakeLinearColor(
     R:NiagaraFloat=&Waveform(
      Global Input Position=&Multiply_Float(A=$Engine.Time, B=1.5)
      Global Phase=$Particles.NormalizedAge
      "Waveform [1]"=::NewEnumerator0
      "[1] Phase"=$Particles.NormalizedAge
      "[1] Frequency"=1.5
      "[1] Amplitude Min/Max"=<1 0>
      "[1] Bias"=$Particles.Desaturation
      "[2] Phase"=$Emitter.Age
      "[3] Phase"=$Emitter.Age
     )
     G:NiagaraFloat=&Waveform(
      Global Input Position=&Multiply_Float(A=$Engine.Time, B=1)
      Global Phase=$Particles.NormalizedAge
      "Waveform [1]"=::NewEnumerator1
      "[1] Phase"=$Particles.NormalizedAge
      "[1] Frequency"=1
      "[1] Amplitude Min/Max"=<1 0>
      "[1] Scale"=2
      "[1] Bias"=$Particles.Desaturation
      "[2] Phase"=$Emitter.Age
      "[3] Phase"=$Emitter.Age
     )
     B:NiagaraFloat=&Waveform(
      Global Input Position=&Multiply_Float(A=$Engine.Time, B=1.5)
      Global Phase=$Particles.NormalizedAge
      "Waveform [1]"=::NewEnumerator2
      "[1] Phase Scale"=2.5
      "[1] Phase"=$Particles.NormalizedAge
      "[1] Frequency"=2
      "[1] Amplitude Min/Max"=<1 0>
      "[1] Bias"=$Particles.Desaturation
      "[2] Phase"=$Emitter.Age
      "[3] Phase"=$Emitter.Age
     )
     A:NiagaraFloat=1
    )) {enabled=false, isSetParameters=true, module=SetVariables_#5}
   /SetVariables(Particles.Color=&MultiplyLinearColor(
     Linear Color=$Particles.WaveColor
     Scale Factor=&MakeVector4_FromLinearColor(LinearColor=&MakeLinearColorFromVectorAndFloat(
       "Vector (RGB)"=<0.2 0.5 1>
       "Float (Alpha)"=&FloatFromCurve(FloatCurve=DataInterface() {propertyValues {Curve {keys [
            {arriveTangent=9.0725, interpMode=RCIM_Cubic, leaveTangent=9.0725, tangentMode=RCTM_User, time=0, value=-0.05}
            {arriveTangent=-2.5249, interpMode=RCIM_Cubic, leaveTangent=-2.5249, tangentMode=RCTM_User, time=0.4, value=1.3}
            {arriveTangent=-2.1667, interpMode=RCIM_Cubic, leaveTangent=-2.1667, time=0.8, value=0.2}
            {arriveTangent=-0.0405, interpMode=RCIM_Cubic, leaveTangent=-0.0405, tangentMode=RCTM_User, time=1, value=0}
           ]}}})
      ))
    )) {enabled=false, isSetParameters=true, module=SetVariables_#6}
   &ScaleMeshSize(Scale Factor=&Multiply_Vector(
     A=&MakeVector(X=$Particles.SizeWave, Y=$Particles.SizeWave, Z=1)
     B=&VectorFromFloat(Value=&FloatFromCurve(
       FloatCurve=DataInterface() {propertyValues {Curve {keys [
           {arriveTangent=4.4833, interpMode=RCIM_Cubic, leaveTangent=4.4833, tangentMode=RCTM_User, time=0, value=0.25}
           {
            arriveTangent=0.0182
            arriveTangentWeight=0.0416
            interpMode=RCIM_Cubic
            leaveTangent=0.0182
            leaveTangentWeight=0.2909
            tangentMode=RCTM_User
            time=0.1
            value=1
           }
           {arriveTangent=0.0045, interpMode=RCIM_Cubic, leaveTangent=0.0045, tangentMode=RCTM_User, time=0.8, value=1}
           {arriveTangent=-0.0405, interpMode=RCIM_Cubic, leaveTangent=-0.0405, tangentMode=RCTM_User, time=1, value=0.1}
          ]}}}
       Scale Curve=1.5
      ))
    )) {module=ScaleMeshSize001}
   &DynamicMaterialParameters(
    Index 0 Param 1=2
    Index 0 Param 2=1
    Module.Param0WriteEnabled=true
    Module.Param1WriteEnabled=true
    Module.Param2WriteEnabled=true
    Module.Param3WriteEnabled=true
   )
   &Color(Color=&LerpLinearColors(
     StartColor=&RandomRangeLinearColor(Minimum=$User.Color_02, Maximum=<rgba 1 0.0396 0.894 1>)
     EndColor=$Particles.Initial.Color
     LerpFactor=&FloatFromCurve(FloatCurve=DataInterface() {propertyValues {Curve {keys [
          {tangentMode=RCTM_User, time=0, value=1}
          {arriveTangent=-0.0405, interpMode=RCIM_Cubic, leaveTangent=-0.0405, tangentMode=RCTM_User, time=0.35, value=1}
          {arriveTangent=-0.8715, interpMode=RCIM_Cubic, leaveTangent=-0.8715, tangentMode=RCTM_User, time=0.8, value=0}
         ]}}})
    ))
   &ScaleColor(
    Color Value To Scale=$Particles.Color
    Scale Alpha=&FloatFromCurve(FloatCurve=DataInterface() {propertyValues {Curve {keys [
         <at 0 0 RCIM_Cubic>
         {arriveTangent=-0.047, interpMode=RCIM_Cubic, leaveTangent=-0.047, tangentMode=RCTM_User, time=0.1, value=1}
         {interpMode=RCIM_Cubic, tangentMode=RCTM_User, time=0.8, value=1}
         {interpMode=RCIM_Cubic, tangentMode=RCTM_User, time=1, value=0}
        ]}}})
    Scale RGB=&VectorFromFloat(Value=2)
    ScaleA=true
    ScaleRGB=true
   )
   /SetVariables(Particles.Color_Wave:LinearColor=&LerpLinearColors(
     StartColor=$Particles.Color
     EndColor=&MultiplyLinearColor(
      Linear Color=$User.Color_02
      Scale Factor=&MakeFloat4FromVectorAndFloat(
       "Vector (RGB)"=<0.8 0.8 0.8>
       "Float (Alpha)"=&FloatFromCurve(FloatCurve=DataInterface() {propertyValues {Curve {keys [
            <at 0 0 RCIM_Cubic>
            {arriveTangent=-0.047, interpMode=RCIM_Cubic, leaveTangent=-0.047, tangentMode=RCTM_User, time=0.1, value=1}
            {interpMode=RCIM_Cubic, tangentMode=RCTM_User, time=0.8, value=1}
            {interpMode=RCIM_Cubic, tangentMode=RCTM_User, time=1, value=0}
           ]}}})
      )
     )
     LerpFactor=&Waveform(
      Global Input Position=$Particles.NormalizedAge
      "Waveform [1]"=::NewEnumerator0
      "[1] Frequency"=4
      "[1] Amplitude Min/Max"=<0 1>
      "[2] Phase"=$Emitter.Age
      "[3] Phase"=$Emitter.Age
     )
    )) {isSetParameters=true, module=SetVariables_#7}
  ]
  rendererClasses [&NiagaraSpriteRendererProperties]
  renderers [{rendererClass=&NiagaraSpriteRendererProperties}]
 }
 Decals001 {
  bEnabled=false
  emitterSpawnScript [/SetVariables(Emitter.UseParticleColorInput=1) {isSetParameters=true, module=SetVariables_#2}]
  emitterUpdateScript [
   &EmitterState()
   &SpawnBurst_Instantaneous(Spawn Count=6, Use Spawn Probability=false)
   &SpawnBurst_Instantaneous(Spawn Count=6, Spawn Time=0.5, Use Spawn Probability=false) {module=SpawnBurst_Instantaneous001}
  ]
  particleSpawnScript [
   &InitializeParticle(
    Color Maximum=$User.Color_03
    Color Minimum=$User.Color_01
    Color Mode=::NewEnumerator2
    Lifetime Max=&Add_Float(A=$User.Duration, B=0)
    Lifetime Min:NiagaraFloat=&Subtract_Float(A=$User.Duration, B=-0.2)
    Lifetime Mode=::NewEnumerator1
    Mesh Scale Mode=::NewEnumerator3
    Mesh Uniform Scale=0.04
    Sprite Size Mode=::NewEnumerator1
    Uniform Sprite Size Max=40
    Uniform Sprite Size Min=30
    UsePositionOffset=false
    Write Material Random=true
   )
   &CameraOffset(Camera Offset Amount=8)
   &FN_SetSpawnGroup(Number of Spawn Groups=6)
   &InitialMeshOrientation(
    Orientation Axis=<0 0 1>
    Orientation Coordinate Space=::World
    Orientation Vector=&Lerp_Vector(
     A=<-1 0 0>
     B=<1 0 0>
     Alpha=&ModuloFloat(A=3, B=&ModuloFloat(A=3, B=&MakeFloatFromInt(Int=$Particles.SpawnGroup)))
    )
    Use Orientation Vector=true
    Use Rotation Vector=false
   )
  ]
  particleUpdateScript [
   &ParticleState()
   &SkeletalMeshLocation(
    Bone / Socket Index=$Particles.SpawnGroup
    Bone Sampling Mode=::NewEnumerator4
    MaskSpawnGroup=false
    Sampled Position Offset=&RandomRangeVector(Minimum=<0 35 6.5>, Maximum=<0 -20 6.5>)
    Skeletal Mesh=DataInterface() {propertyValues {
      DefaultMesh=null
      ExcludeBoneName=root
      OptionalComponentClass=null
      PreviewMesh=/Skyfire/Gameplay/KnockbackWeapon/Meshes/Alien_Knockback_Weapon_Shotgun
      SkinningMode=SkinOnTheFly
      SoftSourceActor=null
      SourceMode=Default
      UvSetIndex=0
      WholeMeshLOD=-1
      bExcludeBone=false
      bReadDeformedGeometry=true
      bRequireCurrentFrameData=true
      ComponentTags []
      FilteredBones [BreakAction]
      FilteredSockets []
      MeshUserParameter {parameter {name=None, typeDefHandle {registeredTypeIndex=1}, varData []}}
      SamplingRegions []
     }}
   )
   /SetVariables(Particles.SizeWave:NiagaraFloat=&Waveform(
     Global Input Position=&Multiply_Float(A=$Engine.Time, B=3)
     Global Phase=$Particles.NormalizedAge
     "Waveform [1]"=::NewEnumerator0
     "[1] Phase Scale"=0
     "[1] Phase"=$Particles.NormalizedAge
     "[1] Frequency"=2
     "[1] Amplitude Min/Max"=<0.2 0>
     "[1] Bias"=1
     "[2] Phase"=$Emitter.Age
     "[3] Phase"=$Emitter.Age
    )) {isSetParameters=true, module=SetVariables_#3}
   /SetVariables(Particles.Desaturation=0.6) {isSetParameters=true, module=SetVariables_#4}
   /SetVariables(Particles.WaveColor:LinearColor=&MakeLinearColor(
     R:NiagaraFloat=&Waveform(
      Global Input Position=&Multiply_Float(A=$Engine.Time, B=1.5)
      Global Phase=$Particles.NormalizedAge
      "Waveform [1]"=::NewEnumerator0
      "[1] Phase"=$Particles.NormalizedAge
      "[1] Frequency"=1.5
      "[1] Amplitude Min/Max"=<1 0>
      "[1] Bias"=$Particles.Desaturation
      "[2] Phase"=$Emitter.Age
      "[3] Phase"=$Emitter.Age
     )
     G:NiagaraFloat=&Waveform(
      Global Input Position=&Multiply_Float(A=$Engine.Time, B=1)
      Global Phase=$Particles.NormalizedAge
      "Waveform [1]"=::NewEnumerator1
      "[1] Phase"=$Particles.NormalizedAge
      "[1] Frequency"=1
      "[1] Amplitude Min/Max"=<1 0>
      "[1] Scale"=2
      "[1] Bias"=$Particles.Desaturation
      "[2] Phase"=$Emitter.Age
      "[3] Phase"=$Emitter.Age
     )
     B:NiagaraFloat=&Waveform(
      Global Input Position=&Multiply_Float(A=$Engine.Time, B=1.5)
      Global Phase=$Particles.NormalizedAge
      "Waveform [1]"=::NewEnumerator2
      "[1] Phase Scale"=2.5
      "[1] Phase"=$Particles.NormalizedAge
      "[1] Frequency"=2
      "[1] Amplitude Min/Max"=<1 0>
      "[1] Bias"=$Particles.Desaturation
      "[2] Phase"=$Emitter.Age
      "[3] Phase"=$Emitter.Age
     )
     A:NiagaraFloat=1
    )) {enabled=false, isSetParameters=true, module=SetVariables_#5}
   /SetVariables(Particles.Color=&MultiplyLinearColor(
     Linear Color=$Particles.WaveColor
     Scale Factor=&MakeVector4_FromLinearColor(LinearColor=&MakeLinearColorFromVectorAndFloat(
       "Vector (RGB)"=<0.2 0.5 1>
       "Float (Alpha)"=&FloatFromCurve(FloatCurve=DataInterface() {propertyValues {Curve {keys [
            {arriveTangent=9.0725, interpMode=RCIM_Cubic, leaveTangent=9.0725, tangentMode=RCTM_User, time=0, value=-0.05}
            {arriveTangent=-2.5249, interpMode=RCIM_Cubic, leaveTangent=-2.5249, tangentMode=RCTM_User, time=0.4, value=1.3}
            {arriveTangent=-2.1667, interpMode=RCIM_Cubic, leaveTangent=-2.1667, time=0.8, value=0.2}
            {arriveTangent=-0.0405, interpMode=RCIM_Cubic, leaveTangent=-0.0405, tangentMode=RCTM_User, time=1, value=0}
           ]}}})
      ))
    )) {enabled=false, isSetParameters=true, module=SetVariables_#6}
   &ScaleMeshSize(Scale Factor=&Multiply_Vector(
     A=&MakeVector(X=$Particles.SizeWave, Y=$Particles.SizeWave, Z=1)
     B=&VectorFromFloat(Value=&FloatFromCurve(
       FloatCurve=DataInterface() {propertyValues {Curve {keys [
           {arriveTangent=4.4833, interpMode=RCIM_Cubic, leaveTangent=4.4833, tangentMode=RCTM_User, time=0, value=0.25}
           {
            arriveTangent=0.0182
            arriveTangentWeight=0.0416
            interpMode=RCIM_Cubic
            leaveTangent=0.0182
            leaveTangentWeight=0.2909
            tangentMode=RCTM_User
            time=0.1
            value=1
           }
           {arriveTangent=0.0045, interpMode=RCIM_Cubic, leaveTangent=0.0045, tangentMode=RCTM_User, time=0.8, value=1}
           {arriveTangent=-0.0405, interpMode=RCIM_Cubic, leaveTangent=-0.0405, tangentMode=RCTM_User, time=1, value=0.1}
          ]}}}
       Scale Curve=1.5
      ))
    )) {module=ScaleMeshSize001}
   &DynamicMaterialParameters(
    Index 0 Param 1=1.25
    Index 0 Param 2=1
    Index 0 Param 3=1
    Module.Param0WriteEnabled=true
    Module.Param1WriteEnabled=true
    Module.Param2WriteEnabled=true
    Module.Param3WriteEnabled=true
   )
   &Color(Color=&LerpLinearColors(
     StartColor=&RandomRangeLinearColor(Minimum=$User.Color_02, Maximum=<rgba 0.5365 0.124 0.4978 1>)
     EndColor=$Particles.Initial.Color
     LerpFactor=&FloatFromCurve(FloatCurve=DataInterface() {propertyValues {Curve {keys [
          {tangentMode=RCTM_User, time=0, value=1}
          {arriveTangent=-0.0405, interpMode=RCIM_Cubic, leaveTangent=-0.0405, tangentMode=RCTM_User, time=0.35, value=1}
          {arriveTangent=-0.8715, interpMode=RCIM_Cubic, leaveTangent=-0.8715, tangentMode=RCTM_User, time=0.8, value=0}
         ]}}})
    ))
   &ScaleColor(
    Color Value To Scale=$Particles.Color
    Scale Alpha=&FloatFromCurve(FloatCurve=DataInterface() {propertyValues {Curve {keys [
         <at 0 0 RCIM_Cubic>
         {arriveTangent=-0.047, interpMode=RCIM_Cubic, leaveTangent=-0.047, tangentMode=RCTM_User, time=0.1, value=1}
         {interpMode=RCIM_Cubic, tangentMode=RCTM_User, time=0.8, value=1}
         {interpMode=RCIM_Cubic, tangentMode=RCTM_User, time=1, value=0}
        ]}}})
    Scale RGB=&VectorFromFloat(Value=2.5)
    ScaleA=true
    ScaleRGB=true
   )
   /SetVariables(Particles.RingColors:LinearColor=&MultiplyLinearColor(Linear Color=$Particles.Color, Scale Factor=<3 3 3 1>)) {isSetParameters=true, module=SetVariables_#8}
  ]
  rendererClasses [&NiagaraSpriteRendererProperties]
  renderers [{rendererClass=&NiagaraSpriteRendererProperties}]
 }
 Lightning_Big {
  emitterUpdateScript [
   &EmitterState()
   &SpawnRate(
    SpawnRate=&Multiply_Float(
     A=&FloatFromCurve(FloatCurve=DataInterface() {propertyValues {Curve {keys [<at 0.95 1>,<at 1 1>]}}}, CurveIndex=$Emitter.NormalizedLoopAge, Scale Curve=60)
     B=&RandomRangeFloat(Minimum=0, Maximum=1, Evaluation Type=::NewEnumerator1)
    )
    Use Spawn Probability=false
   )
  ]
  particleSpawnScript [
   /SetVariables(Particles.MeshOrientation=&MakeQuatFromAxisAngle(
     Axis=&RandomRangeVector(Minimum=<0 0 0>, Maximum=<1 1 1>, Evaluation Type=::NewEnumerator1)
     AngleInDegrees=90
    )) {isSetParameters=true, module=SetVariables_#9}
   /SetVariables(Particles.SpawnRotationVector=&RandomRangeVector(Minimum=<-0.125 -0.125 -0.125>, Maximum=<0.125 0.125 0.125>, Evaluation Type=::NewEnumerator1)) {isSetParameters=true, module=SetVariables_#10}
   /Niagara/Modules/Spawn/Initialization/InitializeParticle(
    Color=$User.Color_03
    Lifetime=&RandomRangeFloat(Minimum=0.1, Maximum=0.3, Evaluation Type=::NewEnumerator1)
    Mesh Scale=&RandomRangeVector(Minimum=<0.05 0.2 0.05>, Maximum=<0.3 0.6 0.3>, Evaluation Type=::NewEnumerator1)
    Module.Write Color=true
    Module.Write Lifetime=true
    Module.Write Mass=true
    Module.Write Material Random=true
    Module.Write Position=false
    Module.Write Scale=true
    Module.Write SpriteRotation=false
    Module.Write SpriteSize=false
   )
   &Color(Color=&MakeLinearColorFromVectorAndFloat(
     "Vector (RGB)"=&RandomRangeVector(Minimum=<0.85 0.7072 0.1353>, Maximum=<0.95 0.8955 0.4908>, Evaluation Type=::NewEnumerator1)
     "Float (Alpha)"=&ClampFloat(Float=1)
    )) {enabled=false}
   &DynamicMaterialParameters(
    Index 0 Param 2=&RandomRangeFloat(Minimum=0, Maximum=1, Evaluation Type=::NewEnumerator1)
    Index 0 Param 4=&RandomRangeFloat(Minimum=0.3, Maximum=0.6, Evaluation Type=::NewEnumerator1)
    Module.Param0WriteEnabled=false
    Module.Param1WriteEnabled=true
    Module.Param2WriteEnabled=false
    Module.Param3WriteEnabled=true
   ) {enabled=false, module=DynamicMaterialParameters001}
  ]
  particleUpdateScript [
   &ParticleState()
   &ScaleColor(
    Scale Alpha=&FloatFromCurve(FloatCurve=DataInterface() {propertyValues {Curve {keys [<at 0.5 1 RCIM_Cubic>,<at 1 0 RCIM_Cubic>]}}})
    ScaleA=true
    ScaleRGB=true
   )
   &SkeletalMeshLocation(Bone Sampling Mode=::NewEnumerator0, MaskSpawnGroup=false, OverrideRandomSeed=false, Skeletal Mesh=$User.SKM_02)
   &UpdateMeshOrientation(Rotation Vector=$Particles.SpawnRotationVector)
   &CameraOffset(Camera Offset Amount=10)
   &DynamicMaterialParameters(
    Index 0 Param 1=&FloatFromCurve(
     FloatCurve=DataInterface() {propertyValues {Curve {keys [{interpMode=RCIM_Cubic, tangentMode=RCTM_User, time=0.5, value=1}, {interpMode=RCIM_Cubic, tangentMode=RCTM_User, time=1, value=0}]}}}
     Scale Curve=1.5
    )
    Index 0 Param 2=0.2
    Index 0 Param 3=3
    Index 0 Param 4=1
    Module.Param0WriteEnabled=true
    Module.Param1WriteEnabled=true
    Module.Param2WriteEnabled=true
    Module.Param3WriteEnabled=true
   )
   &FN_Light_Attributes(
    Color=$User.Color_02
    Create Color Attribute=true
    Create Exponent Attribute=true
    Create Radius Attribute=true
    Enable Particle Light Enable=true
    Exponent=2
    Light Enabled=true
    Radius=30
    Scale Color=20
   )
  ]
  rendererClasses [&NiagaraMeshRendererProperties, &NiagaraLightRendererProperties]
  renderers [
   {rendererClass=&NiagaraMeshRendererProperties, rendererIndex=0}
   {rendererClass=&NiagaraLightRendererProperties, rendererIndex=1}
  ]
 }
 Lightning_small {
  emitterUpdateScript [
   &EmitterState()
   &EmitterScalability(
    Module.bApplySpawnCountScale=false
    Module.bApplySpawnCountScaleByDistanceFraction=false
    Module.bUseMaxDistance=false
    Module.bUseMinDistance=false
   ) {enabled=false}
   &EmitterLifeCycle(CompleteOnInactive=true, MaxLoopCount=1, NextLoopDelay=0.9875, NextLoopDuration=3.55) {enabled=false}
   &SpawnRate(
    SpawnRate=&Multiply_Float(
     A=&FloatFromCurve(FloatCurve=DataInterface() {propertyValues {Curve {keys [<at 0.95 1>,<at 1 1>]}}}, CurveIndex=$Emitter.NormalizedLoopAge, Scale Curve=140)
     B=&RandomRangeFloat(Minimum=0, Maximum=1, Evaluation Type=::NewEnumerator1)
    )
    Use Spawn Probability=false
   )
   &SpawnBurst_Instantaneous(Spawn Count=4, Use Spawn Probability=false) {enabled=false}
  ]
  particleSpawnScript [
   &InitializeParticle(
    Color=<rgba 1 0.0396 0.894 1>
    Lifetime=&UniformRangedFloat(Minimum=0.1, Maximum=0.14)
    Mesh Scale=&Multiply_Vector(
     A=&UniformRangedVector(Minimum=<0.5 0.5 0.5>, Maximum=<0.8 0.8 0.8>)
     B=&VectorFromFloat(Value=1.35)
    )
    Module.Write Color=true
    Module.Write Lifetime=true
    Module.Write Mass=true
    Module.Write Material Random=true
    Module.Write Position=true
    Module.Write Scale=true
    Module.Write SpriteRotation=false
    Module.Write SpriteSize=false
    Position=&SimulationPosition()
    Sprite Size=<25 25>
   )
   /Niagara/Modules/Spawn/Orientation/V2/InitialMeshOrientation(
    Rotation=&UniformRangedVector(Minimum=<-1 0 0>, Maximum=<1 0 0>)
    Use Orientation Vector=true
    Use Rotation Vector=true
   )
   &AddVelocityFromPoint(Module.Use Falloff Distance=false, Velocity Strength=0.01)
  ]
  particleUpdateScript [
   &SolveForcesAndVelocity(Clamp Velocity=false, Limit Acceleration=false)
   &SkeletalMeshLocation(
    Bone Sampling Mode=::NewEnumerator0
    MaskSpawnGroup=false
    OverrideRandomSeed=false
    Sampled Orientation Coordinate Space=::World
    Sampled Position Offset=&RandomRangeVector(Minimum=<-5 -5 -5>, Maximum=<5 5 5>)
    Skeletal Mesh=$User.SKM_02
   )
   /Niagara/Modules/Spawn/Location/SphereLocation(MaskSpawnGroup=false, Module.Override Randomness=false, Offset=<0 0 5>, Sphere Radius=5, Surface Only=true) {enabled=false}
   &ParticleState()
   &ScaleColor(Scale RGB=&VectorFromFloat(Value=50), ScaleA=true, ScaleRGB=true)
   &DynamicMaterialParameters(
    Index 0 Param 1=&UniformRangedFloat(Minimum=4, Maximum=7)
    Index 0 Param 2=&FloatFromCurve(FloatCurve=DataInterface() {propertyValues {Curve {keys [<at 0.1129 1>,<at 1 -1>]}}})
    Index 0 Param 3=&UniformRangedFloat(Minimum=0.9, Maximum=1.5)
    Module.Param0WriteEnabled=true
    Module.Param1WriteEnabled=true
    Module.Param2WriteEnabled=true
    Module.Param3WriteEnabled=true
   )
   /SetVariables(Particles.Light_Size=10) {isSetParameters=true}
   /SetVariables(Particles.Light_Color:LinearColor=&MultiplyLinearColor(Linear Color=$User.Color_02, Scale Factor=&MakeVector4_FromFloat(Float=3))) {isSetParameters=true}
   &CameraOffset(Camera Offset Amount=15)
  ]
  rendererClasses [&NiagaraLightRendererProperties, &NiagaraMeshRendererProperties]
  renderers [
   {rendererClass=&NiagaraLightRendererProperties, rendererIndex=0}
   {rendererClass=&NiagaraMeshRendererProperties, rendererIndex=1}
  ]
 }
 Mesh_Arcs {
  bEnabled=false
  emitterUpdateScript [
   &EmitterState()
   &SpawnRate(
    Spawn Probability=&FloatFromCurve(FloatCurve=DataInterface() {propertyValues {Curve {keys [<at 0 1>,<at 1 0>]}}}, CurveIndex=$Emitter.NormalizedLoopAge)
    SpawnRate=&Multiply_Float(
     A=&FloatFromCurve(
      FloatCurve=DataInterface() {propertyValues {Curve {keys [{interpMode=RCIM_Cubic, tangentMode=RCTM_User, time=0.5, value=1}, {interpMode=RCIM_Cubic, tangentMode=RCTM_User, time=1, value=0}]}}}
      CurveIndex=$Emitter.NormalizedLoopAge
      Scale Curve=20
     )
     B=&RandomRangeFloat(Minimum=0, Maximum=1, Evaluation Type=::NewEnumerator1)
    )
    Use Spawn Probability=false
   )
  ]
  particleSpawnScript [
   /SetVariables(Particles.MeshOrientation=&MakeQuatFromAxisAngle(
     Axis=&RandomRangeVector(Minimum=<-180 -180 -180>, Maximum=<180 180 180>, Evaluation Type=::NewEnumerator1)
     AngleInDegrees=&RandomRangeFloat(Minimum=0, Maximum=360)
    )) {isSetParameters=true}
   /SetVariables(Particles.SpawnRotationVector=&RandomRangeVector(Minimum=<-0.125 -0.125 -0.125>, Maximum=<0.125 0.125 0.125>, Evaluation Type=::NewEnumerator1)) {isSetParameters=true}
   /Niagara/Modules/Spawn/Initialization/V2/InitializeParticle(
    Color Maximum=$User.Color_02
    Color Minimum=$User.Color_01
    Color Mode=::NewEnumerator2
    Lifetime Max=1
    Lifetime Min=0.8
    Lifetime Mode=::NewEnumerator1
    Mesh Scale Mode=::NewEnumerator3
    Mesh Uniform Scale=0.1
    UsePositionOffset=false
    Write Material Random=true
   )
   &SkeletalMeshLocation(
    Bone Sampling Mode=::NewEnumerator0
    MaskSpawnGroup=false
    OverrideRandomSeed=false
    Skeletal Mesh=DataInterface() {propertyValues {
      DefaultMesh=null
      ExcludeBoneName=root
      OptionalComponentClass=null
      PreviewMesh=/Game/Characters/Enemies/Husk/Mesh/Husk_V2/SK_Husk_V2
      SkinningMode=SkinOnTheFly
      SoftSourceActor=null
      SourceMode=Default
      UvSetIndex=0
      WholeMeshLOD=-1
      bExcludeBone=false
      bReadDeformedGeometry=true
      bRequireCurrentFrameData=true
      ComponentTags []
      FilteredBones [pelvis,spine_01,spine_02,spine_03,clavicle_l,upperarm_l,lowerarm_l,hand_l,clavicle_r,upperarm_r,lowerarm_r,hand_r,neck_01,head,thigh_l,calf_l,thigh_r,calf_r]
      FilteredSockets []
      MeshUserParameter {parameter {name=None, typeDefHandle {registeredTypeIndex=1}, varData []}}
      SamplingRegions []
     }}
   ) {enabled=false}
   &DynamicMaterialParameters(
    Index 0 Param 2=&RandomRangeFloat(Minimum=0, Maximum=1, Evaluation Type=::NewEnumerator1)
    Index 0 Param 4=&RandomRangeFloat(Minimum=0.3, Maximum=0.6, Evaluation Type=::NewEnumerator1)
    Module.Param0WriteEnabled=false
    Module.Param1WriteEnabled=true
    Module.Param2WriteEnabled=false
    Module.Param3WriteEnabled=true
   ) {module=DynamicMaterialParameters001}
   &SkeletalMeshLocation(Bone Sampling Mode=::NewEnumerator4, MaskSpawnGroup=false, Orientation Sampling=::NewEnumerator1, Skeletal Mesh=$User.SKM_02) {module=SkeletalMeshLocation002}
   /Niagara/Modules/Spawn/Location/V2/SphereLocation(
    MaskSpawnGroup=false
    Override Seed=false
    Sphere Origin:NiagaraPosition=$Output.SkeletalMeshLocation002.SampledPositionWithoutOffset
    Sphere Radius=25
    Surface Only=false
    UseNonUniformScale=false
    UseOffset=false
    UseSphereOrientation=false
   )
  ]
  particleUpdateScript [
   &ParticleState()
   &SkeletalMeshLocation(MaskSpawnGroup=false, OverrideRandomSeed=false, Skeletal Mesh=$User.SKM_02) {enabled=false, module=SkeletalMeshLocation001}
   &UpdateMeshOrientation(Rotation Vector=$Particles.SpawnRotationVector)
   &CameraOffset(Camera Offset Amount=20)
   &DynamicMaterialParameters(
    Index 0 Param 1=&FloatFromCurve(FloatCurve=DataInterface() {propertyValues {Curve {keys [<at 0 2>,<at 1 -1>]}}})
    Index 0 Param 2=&FloatFromCurve(FloatCurve=DataInterface() {propertyValues {Curve {keys [<at 0 2>,<at 1 -1>]}}}, Scale Curve=0.2)
    Module.Param0WriteEnabled=true
    Module.Param1WriteEnabled=true
    Module.Param2WriteEnabled=true
    Module.Param3WriteEnabled=true
   )
   &Color(Color=&LerpLinearColors(
     StartColor=&RandomRangeLinearColor(Minimum=$User.Color_02, Maximum=<rgba 0 0.8229 0.2574 1>)
     EndColor=$Particles.Initial.Color
     LerpFactor=&FloatFromCurve(FloatCurve=DataInterface() {propertyValues {Curve {keys [
          {tangentMode=RCTM_User, time=0, value=1}
          {arriveTangent=-0.0405, interpMode=RCIM_Cubic, leaveTangent=-0.0405, tangentMode=RCTM_User, time=0.35, value=1}
          {arriveTangent=-0.8715, interpMode=RCIM_Cubic, leaveTangent=-0.8715, tangentMode=RCTM_User, time=0.8, value=0}
         ]}}})
    )) {module=Color001}
   &ScaleColor(
    Color Value To Scale=$Particles.Color
    Scale Alpha=&FloatFromCurve(FloatCurve=DataInterface() {propertyValues {Curve {keys [
         <at 0 0 RCIM_Cubic>
         {arriveTangent=-0.047, interpMode=RCIM_Cubic, leaveTangent=-0.047, tangentMode=RCTM_User, time=0.1, value=1}
         {interpMode=RCIM_Cubic, tangentMode=RCTM_User, time=0.8, value=1}
         {interpMode=RCIM_Cubic, tangentMode=RCTM_User, time=1, value=0}
        ]}}})
    Scale RGB=&VectorFromFloat(Value=5)
    ScaleA=true
    ScaleRGB=true
   )
  ]
  rendererClasses [&NiagaraMeshRendererProperties]
  renderers [{rendererClass=&NiagaraMeshRendererProperties, rendererIndex=0}]
 }
 Mesh_Arcs_Smasher001 {
  bEnabled=false
  emitterUpdateScript [
   &EmitterState()
   &SpawnRate(
    SpawnRate=&Multiply_Float(
     A=&FloatFromCurve(FloatCurve=DataInterface() {propertyValues {Curve {keys [<at 0.5 1>,<at 0.9 0>]}}}, CurveIndex=$Emitter.NormalizedLoopAge, Scale Curve=25)
     B=&RandomRangeFloat(Minimum=0, Maximum=1, Evaluation Type=::NewEnumerator1)
    )
    Use Spawn Probability=false
   )
  ]
  particleSpawnScript [
   /SetVariables(Particles.MeshOrientation=&MakeQuatFromAxisAngle(
     Axis=&RandomRangeVector(Minimum=<0 0 0>, Maximum=<1 1 1>, Evaluation Type=::NewEnumerator1)
     AngleInDegrees=90
    )) {isSetParameters=true, module=SetVariables_#9}
   /SetVariables(Particles.SpawnRotationVector=&RandomRangeVector(Minimum=<-0.125 -0.125 -0.125>, Maximum=<0.125 0.125 0.125>, Evaluation Type=::NewEnumerator1)) {isSetParameters=true, module=SetVariables_#10}
   /Niagara/Modules/Spawn/Initialization/InitializeParticle(
    Color=$User.Color_03
    Lifetime=&RandomRangeFloat(Minimum=0.1, Maximum=0.3, Evaluation Type=::NewEnumerator1)
    Mesh Scale=&RandomRangeVector(Minimum=<0.05 0.05 0.05>, Maximum=<0.2 0.2 0.2>, Evaluation Type=::NewEnumerator1)
    Module.Write Color=true
    Module.Write Lifetime=true
    Module.Write Mass=true
    Module.Write Material Random=true
    Module.Write Position=false
    Module.Write Scale=true
    Module.Write SpriteRotation=false
    Module.Write SpriteSize=false
   )
   &Color(Color=&MakeLinearColorFromVectorAndFloat(
     "Vector (RGB)"=&RandomRangeVector(Minimum=<0.85 0.7072 0.1353>, Maximum=<0.95 0.8955 0.4908>, Evaluation Type=::NewEnumerator1)
     "Float (Alpha)"=&ClampFloat(Float=1)
    )) {enabled=false}
   &SkeletalMeshLocation(Bone Sampling Mode=::NewEnumerator0, MaskSpawnGroup=false, OverrideRandomSeed=false, Skeletal Mesh=$User.SKM_02)
   &DynamicMaterialParameters(
    Index 0 Param 2=&RandomRangeFloat(Minimum=0, Maximum=1, Evaluation Type=::NewEnumerator1)
    Index 0 Param 4=&RandomRangeFloat(Minimum=0.3, Maximum=0.6, Evaluation Type=::NewEnumerator1)
    Module.Param0WriteEnabled=false
    Module.Param1WriteEnabled=true
    Module.Param2WriteEnabled=false
    Module.Param3WriteEnabled=true
   ) {enabled=false, module=DynamicMaterialParameters001}
  ]
  particleUpdateScript [
   &ParticleState()
   &ScaleColor(
    Scale Alpha=&FloatFromCurve(FloatCurve=DataInterface() {propertyValues {Curve {keys [<at 0.5 1 RCIM_Cubic>,<at 1 0 RCIM_Cubic>]}}})
    ScaleA=true
    ScaleRGB=true
   )
   &UpdateMeshOrientation(Rotation Vector=$Particles.SpawnRotationVector)
   &CameraOffset(Camera Offset Amount=15)
   &DynamicMaterialParameters(
    Index 0 Param 1=&FloatFromCurve(FloatCurve=DataInterface() {propertyValues {Curve {keys [<at 0 2>,<at 1 -1>]}}})
    Index 0 Param 2=0.2
    Index 0 Param 3=3
    Index 0 Param 4=1
    Module.Param0WriteEnabled=true
    Module.Param1WriteEnabled=true
    Module.Param2WriteEnabled=true
    Module.Param3WriteEnabled=true
   )
  ]
  rendererClasses [&NiagaraMeshRendererProperties]
  renderers [{rendererClass=&NiagaraMeshRendererProperties, rendererIndex=0}]
 }
 PostProcess {
  bEnabled=false
  emitterUpdateScript [
   &EmitterState()
   &SpawnBurst_Instantaneous(Spawn Count=1, Use Spawn Probability=false)
  ]
  particleSpawnScript [/Niagara/Modules/Spawn/Initialization/V2/InitializeParticle(
    Lifetime Max=&Add_Float(A=$User.Duration, B=0.2)
    Lifetime Min:NiagaraFloat=&Subtract_Float(A=$User.Duration, B=0.2)
    Lifetime Mode=::NewEnumerator1
    Mass Max=2
    Mass Mode=::NewEnumerator3
    Sprite Size Max=<3 14>
    Sprite Size Min=<2 6>
    Sprite Size Mode=::NewEnumerator2
    UsePositionOffset=false
    Write Material Random=true
   )]
  particleUpdateScript [
   &ParticleState()
   &SkeletalMeshLocation(Bone Sampling Mode=::NewEnumerator4, MaskSpawnGroup=false, Skeletal Mesh=$User.SKM_02)
   /Game/Effects/Niagara/Modules/ComponentRenderer/FN_PostProcess(Blend:NiagaraFloat=&FloatFromCurve(FloatCurve=DataInterface() {propertyValues {Curve {keys [
         {interpMode=RCIM_Cubic, tangentMode=RCTM_User, time=0, value=0}
         {interpMode=RCIM_Cubic, tangentMode=RCTM_User, time=0.1, value=1}
         {interpMode=RCIM_Cubic, tangentMode=RCTM_User, time=0.8, value=1}
         {interpMode=RCIM_Cubic, tangentMode=RCTM_User, time=1, value=0}
        ]}}}))
   /SetVariables(Particles.ComponentsEnabled:NiagaraBool=&SetBoolByFloatComparison(
     A=1
     B=&FloatFromCurve(FloatCurve=DataInterface() {propertyValues {Curve {keys [<at 0 0>,<at 0.1 1>,<at 0.9 1>,<at 1 0>]}}})
    )) {isSetParameters=true}
  ]
  rendererClasses [&NiagaraComponentRendererProperties]
  renderers [{rendererClass=&NiagaraComponentRendererProperties, rendererIndex=0}]
 }
 Ring_Burst001 {
  bEnabled=false
  emitterSpawnScript [/SetVariables(Emitter.UseParticleColorInput=1) {isSetParameters=true, module=SetVariables_#2}]
  emitterUpdateScript [
   &EmitterState()
   &SpawnBurst_Instantaneous(Spawn Count=12, Use Spawn Probability=false)
   &SpawnBurst_Instantaneous(Spawn Count=6, Spawn Time=0.5, Use Spawn Probability=false) {enabled=false, module=SpawnBurst_Instantaneous001}
  ]
  particleSpawnScript [
   &InitializeParticle(
    Color Maximum=$User.Color_03
    Color Minimum=$User.Color_01
    Color Mode=::NewEnumerator2
    Lifetime Max=&Add_Float(A=$User.Duration, B=0.2)
    Lifetime Min:NiagaraFloat=&Subtract_Float(A=$User.Duration, B=0.2)
    Lifetime Mode=::NewEnumerator1
    Mesh Scale Mode=::NewEnumerator3
    Mesh Uniform Scale=0.06
    Sprite Size Mode=::NewEnumerator1
    Uniform Sprite Size Max=40
    Uniform Sprite Size Min=30
    UsePositionOffset=false
    Write Material Random=true
   )
   &CameraOffset(Camera Offset Amount=8)
   &FN_SetSpawnGroup(Number of Spawn Groups=6)
   /Niagara/Modules/Spawn/Orientation/V3/InitialMeshOrientation(
    Orientation Axis=<0 0 1>
    Orientation Coordinate Space=::World
    Orientation Vector=<0 1 0>
    Use Orientation Vector=true
    Use Rotation Vector=false
   )
  ]
  particleUpdateScript [
   &ParticleState()
   &SkeletalMeshLocation(
    Bone / Socket Index=$Particles.SpawnGroup
    Bone Sampling Mode=::NewEnumerator4
    MaskSpawnGroup=false
    Sampled Position Offset=&RandomRangeVector(Minimum=<0 35 6.5>, Maximum=<0 -20 6.5>)
    Skeletal Mesh=DataInterface() {propertyValues {
      DefaultMesh=null
      ExcludeBoneName=root
      OptionalComponentClass=null
      PreviewMesh=&Alien_Knockback_Weapon_Shotgun
      SkinningMode=SkinOnTheFly
      SoftSourceActor=null
      SourceMode=Default
      UvSetIndex=0
      WholeMeshLOD=-1
      bExcludeBone=false
      bReadDeformedGeometry=true
      bRequireCurrentFrameData=true
      ComponentTags []
      FilteredBones [BreakAction]
      FilteredSockets []
      MeshUserParameter {parameter {name=None, typeDefHandle {registeredTypeIndex=1}, varData []}}
      SamplingRegions []
     }}
   )
   /SetVariables(Particles.SizeWave:NiagaraFloat=&Waveform(
     Global Input Position=&Multiply_Float(A=$Engine.Time, B=3)
     Global Phase=$Particles.NormalizedAge
     "Waveform [1]"=::NewEnumerator0
     "[1] Phase Scale"=0
     "[1] Phase"=$Particles.NormalizedAge
     "[1] Frequency"=2
     "[1] Amplitude Min/Max"=<0.2 0>
     "[1] Bias"=1
     "[2] Phase"=$Emitter.Age
     "[3] Phase"=$Emitter.Age
    )) {isSetParameters=true, module=SetVariables_#3}
   /SetVariables(Particles.Desaturation=0.6) {isSetParameters=true, module=SetVariables_#4}
   /SetVariables(Particles.WaveColor:LinearColor=&MakeLinearColor(
     R:NiagaraFloat=&Waveform(
      Global Input Position=&Multiply_Float(A=$Engine.Time, B=1.5)
      Global Phase=$Particles.NormalizedAge
      "Waveform [1]"=::NewEnumerator0
      "[1] Phase"=$Particles.NormalizedAge
      "[1] Frequency"=1.5
      "[1] Amplitude Min/Max"=<1 0>
      "[1] Bias"=$Particles.Desaturation
      "[2] Phase"=$Emitter.Age
      "[3] Phase"=$Emitter.Age
     )
     G:NiagaraFloat=&Waveform(
      Global Input Position=&Multiply_Float(A=$Engine.Time, B=1)
      Global Phase=$Particles.NormalizedAge
      "Waveform [1]"=::NewEnumerator1
      "[1] Phase"=$Particles.NormalizedAge
      "[1] Frequency"=1
      "[1] Amplitude Min/Max"=<1 0>
      "[1] Scale"=2
      "[1] Bias"=$Particles.Desaturation
      "[2] Phase"=$Emitter.Age
      "[3] Phase"=$Emitter.Age
     )
     B:NiagaraFloat=&Waveform(
      Global Input Position=&Multiply_Float(A=$Engine.Time, B=1.5)
      Global Phase=$Particles.NormalizedAge
      "Waveform [1]"=::NewEnumerator2
      "[1] Phase Scale"=2.5
      "[1] Phase"=$Particles.NormalizedAge
      "[1] Frequency"=2
      "[1] Amplitude Min/Max"=<1 0>
      "[1] Bias"=$Particles.Desaturation
      "[2] Phase"=$Emitter.Age
      "[3] Phase"=$Emitter.Age
     )
     A:NiagaraFloat=1
    )) {enabled=false, isSetParameters=true, module=SetVariables_#5}
   /SetVariables(Particles.Color=&MultiplyLinearColor(
     Linear Color=$Particles.WaveColor
     Scale Factor=&MakeVector4_FromLinearColor(LinearColor=&MakeLinearColorFromVectorAndFloat(
       "Vector (RGB)"=<0.2 0.5 1>
       "Float (Alpha)"=&FloatFromCurve(FloatCurve=DataInterface() {propertyValues {Curve {keys [
            {arriveTangent=9.0725, interpMode=RCIM_Cubic, leaveTangent=9.0725, tangentMode=RCTM_User, time=0, value=-0.05}
            {arriveTangent=-2.5249, interpMode=RCIM_Cubic, leaveTangent=-2.5249, tangentMode=RCTM_User, time=0.4, value=1.3}
            {arriveTangent=-2.1667, interpMode=RCIM_Cubic, leaveTangent=-2.1667, time=0.8, value=0.2}
            {arriveTangent=-0.0405, interpMode=RCIM_Cubic, leaveTangent=-0.0405, tangentMode=RCTM_User, time=1, value=0}
           ]}}})
      ))
    )) {enabled=false, isSetParameters=true, module=SetVariables_#6}
   &ScaleMeshSize(Scale Factor=&Multiply_Vector(
     A=&MakeVector(X=$Particles.SizeWave, Y=$Particles.SizeWave, Z=1)
     B=&VectorFromFloat(Value=&FloatFromCurve(
       FloatCurve=DataInterface() {propertyValues {Curve {keys [
           {arriveTangent=4.4833, interpMode=RCIM_Cubic, leaveTangent=4.4833, tangentMode=RCTM_User, time=0, value=0.25}
           {
            arriveTangent=0.0182
            arriveTangentWeight=0.0416
            interpMode=RCIM_Cubic
            leaveTangent=0.0182
            leaveTangentWeight=0.2909
            tangentMode=RCTM_User
            time=0.1
            value=1
           }
           {arriveTangent=0.0045, interpMode=RCIM_Cubic, leaveTangent=0.0045, tangentMode=RCTM_User, time=0.8, value=1}
           {arriveTangent=-0.0405, interpMode=RCIM_Cubic, leaveTangent=-0.0405, tangentMode=RCTM_User, time=1, value=0.1}
          ]}}}
       Scale Curve=1.5
      ))
    )) {module=ScaleMeshSize001}
   &DynamicMaterialParameters(
    Index 0 Param 1=&FloatFromCurve(
     FloatCurve=DataInterface() {propertyValues {Curve {keys [
         {arriveTangent=9.0725, interpMode=RCIM_Cubic, leaveTangent=9.0725, tangentMode=RCTM_User, time=0, value=0}
         {arriveTangent=0.1068, interpMode=RCIM_Cubic, leaveTangent=0.1068, tangentMode=RCTM_User, time=0.1515, value=0.9243}
         <at 0.7 1>
         {arriveTangent=-0.0405, interpMode=RCIM_Cubic, leaveTangent=-0.0405, tangentMode=RCTM_User, time=1, value=0}
        ]}}}
     Scale Curve=0.35
    )
    Index 0 Param 2=1
    Module.Param0WriteEnabled=true
    Module.Param1WriteEnabled=true
    Module.Param2WriteEnabled=true
    Module.Param3WriteEnabled=true
   )
   &Color(Color=&LerpLinearColors(
     StartColor=&RandomRangeLinearColor(Minimum=$User.Color_02, Maximum=<rgba 0.5365 0.124 0.4978 1>)
     EndColor=$Particles.Initial.Color
     LerpFactor=&FloatFromCurve(FloatCurve=DataInterface() {propertyValues {Curve {keys [
          {tangentMode=RCTM_User, time=0, value=1}
          {arriveTangent=-0.0405, interpMode=RCIM_Cubic, leaveTangent=-0.0405, tangentMode=RCTM_User, time=0.35, value=1}
          {arriveTangent=-0.8715, interpMode=RCIM_Cubic, leaveTangent=-0.8715, tangentMode=RCTM_User, time=0.8, value=0}
         ]}}})
    ))
   &ScaleColor(
    Color Value To Scale=$Particles.Color
    Scale Alpha=&FloatFromCurve(FloatCurve=DataInterface() {propertyValues {Curve {keys [
         <at 0 0 RCIM_Cubic>
         {arriveTangent=-0.047, interpMode=RCIM_Cubic, leaveTangent=-0.047, tangentMode=RCTM_User, time=0.1, value=1}
         {interpMode=RCIM_Cubic, tangentMode=RCTM_User, time=0.8, value=1}
         {interpMode=RCIM_Cubic, tangentMode=RCTM_User, time=1, value=0}
        ]}}})
    ScaleA=true
    ScaleRGB=true
   )
   /SetVariables(Particles.RingColors:LinearColor=&MultiplyLinearColor(Linear Color=$Particles.Color, Scale Factor=<3 3 3 1>)) {isSetParameters=true, module=SetVariables_#8}
  ]
  rendererClasses [&NiagaraMeshRendererProperties]
  renderers [{rendererClass=&NiagaraMeshRendererProperties, rendererIndex=0}]
 }
 Sphere {
  bEnabled=false
  emitterSpawnScript [/SetVariables(Emitter.UseParticleColorInput=1) {isSetParameters=true, module=SetVariables_#2}]
  emitterUpdateScript [
   &EmitterState()
   &SpawnBurst_Instantaneous(Spawn Count=1, Use Spawn Probability=false)
   &SpawnBurst_Instantaneous(Spawn Count=6, Spawn Time=0.5, Use Spawn Probability=false) {enabled=false, module=SpawnBurst_Instantaneous001}
  ]
  particleSpawnScript [
   &InitializeParticle(
    Color Maximum=$User.Color_03
    Color Minimum=$User.Color_01
    Color Mode=::NewEnumerator2
    Lifetime Max=&Add_Float(A=$User.Duration, B=0.2)
    Lifetime Min:NiagaraFloat=&Subtract_Float(A=$User.Duration, B=0.2)
    Lifetime Mode=::NewEnumerator1
    Mesh Scale=<1.6 1 0.8>
    Mesh Scale Mode=::NewEnumerator4
    Position Offset Coordinate Space=::World
    Sprite Size Mode=::NewEnumerator1
    Uniform Sprite Size Max=80
    Uniform Sprite Size Min=70
    UsePositionOffset=true
    Write Material Random=true
   )
   &CameraOffset(Camera Offset Amount=8)
   &InitialMeshOrientation(
    Orientation Axis=<0 0 1>
    Orientation Coordinate Space=::World
    Orientation Vector=&Lerp_Vector(
     A=<-1 0 0>
     B=<1 0 0>
     Alpha=&ModuloFloat(A=3, B=&ModuloFloat(A=3, B=&MakeFloatFromInt(Int=$Particles.SpawnGroup)))
    )
    Use Orientation Vector=true
    Use Rotation Vector=false
   ) {enabled=false}
  ]
  particleUpdateScript [
   &ParticleState()
   &SkeletalMeshLocation(Bone Sampling Mode=::NewEnumerator4, MaskSpawnGroup=false, Skeletal Mesh=$User.SKM_02)
   /SetVariables(Particles.SizeWave:NiagaraFloat=&Waveform(
     Global Input Position=&Multiply_Float(A=$Engine.Time, B=3)
     Global Phase=$Particles.NormalizedAge
     "Waveform [1]"=::NewEnumerator0
     "[1] Phase Scale"=0
     "[1] Phase"=$Particles.NormalizedAge
     "[1] Frequency"=2
     "[1] Amplitude Min/Max"=<0.05 0>
     "[1] Bias"=1
     "[2] Phase"=$Emitter.Age
     "[3] Phase"=$Emitter.Age
    )) {isSetParameters=true, module=SetVariables_#3}
   /SetVariables() {isSetParameters=true, module=SetVariables_#4}
   /SetVariables(Particles.WaveColor:LinearColor=&MakeLinearColor(
     R:NiagaraFloat=&Waveform(
      Global Input Position=&Multiply_Float(A=$Engine.Time, B=1.5)
      Global Phase=$Particles.NormalizedAge
      "Waveform [1]"=::NewEnumerator0
      "[1] Phase"=$Particles.NormalizedAge
      "[1] Frequency"=1.5
      "[1] Amplitude Min/Max"=<1 0>
      "[1] Bias"=$Particles.Desaturation
      "[2] Phase"=$Emitter.Age
      "[3] Phase"=$Emitter.Age
     )
     G:NiagaraFloat=&Waveform(
      Global Input Position=&Multiply_Float(A=$Engine.Time, B=1)
      Global Phase=$Particles.NormalizedAge
      "Waveform [1]"=::NewEnumerator1
      "[1] Phase"=$Particles.NormalizedAge
      "[1] Frequency"=1
      "[1] Amplitude Min/Max"=<1 0>
      "[1] Scale"=2
      "[1] Bias"=$Particles.Desaturation
      "[2] Phase"=$Emitter.Age
      "[3] Phase"=$Emitter.Age
     )
     B:NiagaraFloat=&Waveform(
      Global Input Position=&Multiply_Float(A=$Engine.Time, B=1.5)
      Global Phase=$Particles.NormalizedAge
      "Waveform [1]"=::NewEnumerator2
      "[1] Phase Scale"=2.5
      "[1] Phase"=$Particles.NormalizedAge
      "[1] Frequency"=2
      "[1] Amplitude Min/Max"=<1 0>
      "[1] Bias"=$Particles.Desaturation
      "[2] Phase"=$Emitter.Age
      "[3] Phase"=$Emitter.Age
     )
     A:NiagaraFloat=1
    )) {enabled=false, isSetParameters=true, module=SetVariables_#5}
   /SetVariables(Particles.Color=&MultiplyLinearColor(
     Linear Color=$Particles.WaveColor
     Scale Factor=&MakeVector4_FromLinearColor(LinearColor=&MakeLinearColorFromVectorAndFloat(
       "Vector (RGB)"=<0.2 0.5 1>
       "Float (Alpha)"=&FloatFromCurve(FloatCurve=DataInterface() {propertyValues {Curve {keys [
            {arriveTangent=9.0725, interpMode=RCIM_Cubic, leaveTangent=9.0725, tangentMode=RCTM_User, time=0, value=-0.05}
            {arriveTangent=-2.5249, interpMode=RCIM_Cubic, leaveTangent=-2.5249, tangentMode=RCTM_User, time=0.4, value=1.3}
            {arriveTangent=-2.1667, interpMode=RCIM_Cubic, leaveTangent=-2.1667, time=0.8, value=0.2}
            {arriveTangent=-0.0405, interpMode=RCIM_Cubic, leaveTangent=-0.0405, tangentMode=RCTM_User, time=1, value=0}
           ]}}})
      ))
    )) {enabled=false, isSetParameters=true, module=SetVariables_#6}
   &ScaleMeshSize(Scale Factor=&Multiply_Vector(
     A=&MakeVector(X=$Particles.SizeWave, Y=$Particles.SizeWave, Z=1)
     B=&VectorFromFloat(Value=&FloatFromCurve(
       FloatCurve=DataInterface() {propertyValues {Curve {keys [
           {arriveTangent=4.4833, interpMode=RCIM_Cubic, leaveTangent=4.4833, tangentMode=RCTM_User, time=0, value=0.25}
           {
            arriveTangent=0.0182
            arriveTangentWeight=0.0416
            interpMode=RCIM_Cubic
            leaveTangent=0.0182
            leaveTangentWeight=0.2909
            tangentMode=RCTM_User
            time=0.05
            value=1
           }
           {arriveTangent=0.0045, interpMode=RCIM_Cubic, leaveTangent=0.0045, tangentMode=RCTM_User, time=0.8, value=1}
          ]}}}
       Scale Curve=1.5
      ))
    )) {module=ScaleMeshSize001}
   &DynamicMaterialParameters(
    Index 0 Param 2=1
    Index 0 Param 3=2
    Module.Param0WriteEnabled=true
    Module.Param1WriteEnabled=true
    Module.Param2WriteEnabled=true
    Module.Param3WriteEnabled=true
   )
   &Color(Color=&LerpLinearColors(
     StartColor=&RandomRangeLinearColor(Minimum=$User.Color_02, Maximum=<rgba 0 0.8229 0.2574 1>)
     EndColor=$Particles.Initial.Color
     LerpFactor=&FloatFromCurve(
      FloatCurve=DataInterface() {propertyValues {Curve {keys [
          {tangentMode=RCTM_User, time=0, value=1}
          {arriveTangent=-0.0405, interpMode=RCIM_Cubic, leaveTangent=-0.0405, tangentMode=RCTM_User, time=0.35, value=1}
          {arriveTangent=-0.8715, interpMode=RCIM_Cubic, leaveTangent=-0.8715, tangentMode=RCTM_User, time=0.8, value=0}
         ]}}}
      Scale Curve=0.5
     )
    ))
   &ScaleColor(
    Color Value To Scale=$Particles.Color
    Scale Alpha=&FloatFromCurve(FloatCurve=DataInterface() {propertyValues {Curve {keys [
         <at 0 0 RCIM_Cubic>
         {arriveTangent=-0.047, interpMode=RCIM_Cubic, leaveTangent=-0.047, tangentMode=RCTM_User, time=0.1, value=1}
         {interpMode=RCIM_Cubic, tangentMode=RCTM_User, time=0.8, value=1}
         {interpMode=RCIM_Cubic, tangentMode=RCTM_User, time=1, value=0}
        ]}}})
    Scale RGB=&VectorFromFloat(Value=5)
    ScaleA=true
    ScaleRGB=true
   )
   /SetVariables(Particles.Color_Wave:LinearColor=&LerpLinearColors(
     StartColor=$Particles.Color
     EndColor=&MultiplyLinearColor(
      Linear Color=$User.Color_02
      Scale Factor=&MakeFloat4FromVectorAndFloat(
       "Vector (RGB)"=<0.8 0.8 0.8>
       "Float (Alpha)"=&FloatFromCurve(FloatCurve=DataInterface() {propertyValues {Curve {keys [
            <at 0 0 RCIM_Cubic>
            {arriveTangent=-0.047, interpMode=RCIM_Cubic, leaveTangent=-0.047, tangentMode=RCTM_User, time=0.1, value=1}
            {interpMode=RCIM_Cubic, tangentMode=RCTM_User, time=0.8, value=1}
            {interpMode=RCIM_Cubic, tangentMode=RCTM_User, time=1, value=0}
           ]}}})
      )
     )
     LerpFactor=&Waveform(
      Global Input Position=$Particles.NormalizedAge
      "Waveform [1]"=::NewEnumerator0
      "[1] Frequency"=2
      "[1] Amplitude Min/Max"=<0 1>
      "[2] Phase"=$Emitter.Age
      "[3] Phase"=$Emitter.Age
     )
    )) {isSetParameters=true, module=SetVariables_#7}
   /SetVariables(Particles.Pos2:Vector3f=&Add_Vector(A=Linked() {linkedVariable {name=Particles.Position, type=NiagaraPosition}}, B=<0 0 -11>)) {isSetParameters=true}
  ]
  rendererClasses [&NiagaraMeshRendererProperties]
  renderers [{rendererClass=&NiagaraMeshRendererProperties, rendererIndex=0}]
 }
}
summary {
 systemName=NS_Knockback_PlayerDampen
 emitters [
  {emitterName=Decals, rendererClasses [&NiagaraSpriteRendererProperties]}
  {bEnabled=false, emitterName=Ring_Burst001, rendererClasses [&NiagaraMeshRendererProperties]}
  {bEnabled=false, emitterName=Decals001, rendererClasses [&NiagaraSpriteRendererProperties]}
  {bEnabled=false, emitterName=PostProcess, rendererClasses [&NiagaraComponentRendererProperties]}
  {bEnabled=false, emitterName=Sphere, rendererClasses [&NiagaraMeshRendererProperties]}
  {bEnabled=false, emitterName=Mesh_Arcs, rendererClasses [&NiagaraMeshRendererProperties]}
  {emitterName=Lightning_Big, rendererClasses [&NiagaraMeshRendererProperties, &NiagaraLightRendererProperties]}
  {bEnabled=false, emitterName=Mesh_Arcs_Smasher001, rendererClasses [&NiagaraMeshRendererProperties]}
  {emitterName=Lightning_small, rendererClasses [&NiagaraLightRendererProperties, &NiagaraMeshRendererProperties]}
 ]
 userVariables [
  User.Color_01 {type=LinearColor, description="", defaultValue {struct=LinearColor, value=<rgba 1 0.0396 0.894 1>}}
  User.Color_02 {type=LinearColor, description="", defaultValue {struct=LinearColor, value=<rgba 0.8632 0.3712 1 1>}}
  User.Color_03 {type=LinearColor, description="", defaultValue {struct=LinearColor, value=<rgba 1 0.0272 0.2379 1>}}
  User.Duration {type=NiagaraFloat, defaultValue=2, description=""}
  User.SKM_02 {
   type=NiagaraDataInterfaceSkeletalMesh
   defaultValue {
    struct=NiagaraExt_VariableValue_DataInterface
    value {dataInterface=&self:NiagaraDataInterfaceSkeletalMesh_1}
   }
  }
 ]
}
